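# Volumetric cloud material for the Clyde time-of-day setup, rendering the cloud layer and the safe-zone storm ring.
# SDF volume-texture clouds are eroded by noise and colored by time of day, with parameter-driven Hades and Kelp event overrides.
# The storm shapes a profile texture using safe-zone collection values, adding Voronoi noise, rotating 3D flowmaps and masks that blend into the clouds.
# Conservative density, empty-space skipping and simplified reflection-capture branches limit cost.
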Material /Clyde/TimeOfDay/TODM/TODS_Clyde/MV_Clyde_VolumetricCloudsV2
graph {
 properties {
  bUsedWithVolumetricCloud=true
  blendMode=BLEND_Additive
  materialDomain=MD_Volume
  pixelDepthOffsetMode=PDOM_Legacy
  usedShadingModels=Default Lit
 }
 nodes [
  Constant() {r=1}
  Constant()
  Constant()
  n1 = NamedRerouteUsage()
  n2 = ScalarParameter() {defaultValue=0.1, desc=Used to reduce the glow on the emissive for kelp, group=Kelp_Clouds, parameterName=Kelp_Backdrop_Glow_Multi}
  n3 = NamedRerouteUsage()
  n4 = NamedRerouteUsage()
  n5 = Max(
   A=CollectionParameter() {collection=/Game/Athena/SafeZone/MPC_SafeZoneClouds, parameterName=CloudStartingBias}
   B=CollectionParameter() {collection=&MPC_SafeZoneClouds, parameterName=PreventStormShrinking}
  )
  n6 = Multiply(
   A=CollectionParameter() {collection=&MPC_SafeZoneClouds, parameterName=EnableVolumetricStorm}
   B=Saturate(None=OneMinus(None=@n5))
  )
  LinearInterpolate(
   A=@n1
   B=Multiply(A=VectorParameter() {group=Kelp_Clouds, parameterName=Kelp_Backdrop_Glow_Volumetric, defaultValue=<rgba 10 10 10 1>}.RGB, B=@n2)
   Alpha=LinearInterpolate(A=Multiply(A=@n3, B=@n4), Alpha=Saturate(None=@n6)) {constB=0}
  )
  n7 = VectorParameter() {parameterName=Cloud Color Mult, defaultValue=<rgba 1 1 1 0>}
  n8 = NamedRerouteUsage()
  n9 = NamedRerouteUsage()
  LinearInterpolate(
   A=Constant3Vector() {constant=<rgba 1 1 1 0>}
   B=@n7.RGB
   Alpha=NamedRerouteDeclaration(None=Multiply(A=@n8, B=@n9)) {name=Storm_Kelp_Color_Mask, nodeColor=<rgba 0 1 0.3438 1>}
  )
  n10 = VectorParameter() {group=Clouds, parameterName=AlbedoCloudsMorning, defaultValue=<rgba 1 1 1 0>}
  LinearInterpolate(A=Constant3Vector() {constant=<rgba 0.15 0.1046 0.2122 1>}, B=@n10.RGB, Alpha=NamedRerouteUsage())
  n11 = VectorParameter() {group=Clouds, parameterName=AlbedoCloudsDay, defaultValue=<rgba 1 1 1 0>}
  LinearInterpolate(A=Constant3Vector() {constant=<rgba 1 1 1 0>}, B=@n11.RGB, Alpha=NamedRerouteUsage())
  n12 = VectorParameter() {group=Clouds, parameterName=AlbedoCloudsEvening, defaultValue=<rgba 1 1 1 0>}
  LinearInterpolate(A=Constant3Vector() {constant=<rgba 0.6795 0.4969 0.3419 1>}, B=@n12.RGB, Alpha=NamedRerouteUsage())
  n13 = VectorParameter() {group=Clouds, parameterName=AlbedoCloudsNight, defaultValue=<rgba 1 1 1 0>}
  LinearInterpolate(A=Constant3Vector() {constant=<rgba 1 1 1 0>}, B=@n13.RGB, Alpha=NamedRerouteUsage())
  n14 = VectorParameter() {group=Clouds, parameterName=CloudsEmissiveMorning, defaultValue=<rgba 0 0 0 1>}
  LinearInterpolate(A=Constant3Vector() {constant=<rgba 0.1697 0.1552 0.2 1>}, B=@n14.RGB, Alpha=NamedRerouteUsage())
  n15 = VectorParameter() {group=Clouds, parameterName=CloudsEmissiveDay, defaultValue=<rgba 0.005 0.005 0.005 1>}
  LinearInterpolate(A=Constant3Vector() {constant=<rgba 0.1 0.1 0.1 1>}, B=@n15.RGB, Alpha=NamedRerouteUsage())
  n16 = VectorParameter() {group=Clouds, parameterName=CloudsEmissiveEvening, defaultValue=<rgba 0.002 0.0026 0.009 1>}
  LinearInterpolate(A=Constant3Vector() {constant=<rgba 0.1301 0.0931 0.15 1>}, B=@n16.RGB, Alpha=NamedRerouteUsage())
  n17 = VectorParameter() {group=Clouds, parameterName=CloudsEmissiveNight, defaultValue=<rgba 0.0009 0.0067 0.007 1>}
  LinearInterpolate(A=Constant3Vector() {constant=<rgba 0.0144 0.0168 0.0252 1>}, B=@n17.RGB, Alpha=NamedRerouteUsage())
  n18 = MaterialFunctionCall(Float2=ScreenPosition().ViewportUV) {materialFunction=&BreakOutFloat2Components}
  n19 = MaterialFunctionCall(Float2=ViewSize()) {materialFunction=&BreakOutFloat2Components}
  n20 = Divide(A=@n19.G, B=@n19.G)
  n21 = OneMinus(None=@n20)
  MaterialFunctionCall(X=@n18.R, Y=Add(A=Multiply(A=@n18.G, B=@n20), B=Divide(A=@n21))) {materialFunction=&MakeFloat2}
  n22 = Subtract(A=ComponentMask(None=VolumetricAdvancedMaterialInput().ConservativeDensity as Float4) {g=true}, B=NamedRerouteUsage())
  MaterialFunctionCall(
   Input=@n22
   Input Low=ScalarParameter() {defaultValue=0.5, parameterName=SDF_Max}
   Input High=ScalarParameter() {defaultValue=-0.1667, parameterName=SDF_Min}
  ) {materialFunction=&RemapValue_0-1}
  n23 = Time()
  n24 = CollectionParameter() {collection=&MPC_SafeZoneClouds, parameterName=CustomCloudsTime}
  n25 = Ceil(None=Saturate(None=@n24))
  Multiply(
   A=NamedRerouteDeclaration(None=LinearInterpolate(A=@n23, B=@n24, Alpha=@n25)) {name=CustomTime, nodeColor=<rgba 0 1 0.1094 1>}
   B=CollectionParameter() {collection=&MPC_SafeZoneClouds, parameterName=TopCloudsSpeed}
  )
  n26 = NamedRerouteUsage()
  n27 = NamedRerouteUsage()
  n28 = Multiply(A=NamedRerouteDeclaration(None=Multiply(A=@n26, B=@n27)) {name=Clouds, nodeColor=<rgba 0 1 0.5078 1>}, B=NamedRerouteUsage())
  NamedRerouteDeclaration(None=Max(A=ReflectionCapturePassSwitch!(Default=@n28, Reflection=@n27), B=NamedRerouteUsage())) {name=Extinction, nodeColor=<rgba 0.9922 0 1 1>}
  n29 = CollectionParameter() {collection=&FortniteMaterialParameters, parameterName=HadesActive}
  NamedRerouteDeclaration(None=LinearInterpolate(
    A=ScalarParameter() {defaultValue=0.5, parameterName=SDF_Max}
    B=ScalarParameter() {defaultValue=1.3, group=HadesClouds, parameterName=Hades_SDF_Max}
    Alpha=@n29
   )) {name=SDF_Max, nodeColor=<rgba 1 0 0.6797 1>}
  NamedRerouteDeclaration(None=LinearInterpolate(
    A=ScalarParameter() {defaultValue=0.1, desc=0.1, parameterName=CloudNoiseGlobalStrength}
    B=ScalarParameter() {defaultValue=2.5, group=HadesClouds, parameterName=Hades_Noise_Strength}
    Alpha=@n29
   )) {name=NoiseStrength, nodeColor=<rgba 1 0 0.3516 1>}
  n30 = Distance(
   A=ComponentMask(None=WorldPosition().XYZ) {g=true, r=true}
   B=ComponentMask(None=CollectionParameter() {collection=&MPC_SafeZoneClouds, parameterName=SafeZoneLoc}) {g=true, r=true}
  )
  n31 = MaterialFunctionCall(Float2=NamedRerouteUsage()) {materialFunction=&BreakOutFloat2Components}
  n32 = MaterialFunctionCall!(Km=@n31.G) {materialFunction=/Engine/Functions/Engine_MaterialFunctions01/Units/Km-to-Cm}
  n33 = ScalarParameter() {defaultValue=1, group=StormRadialWaves, parameterName=WavesInnerRadiusMultiplier}
  n34 = Subtract(A=@n30, B=@n32.Cm)
  n35 = MaterialFunctionCall(Km=@n31.R) {materialFunction=&Km-to-Cm}
  n36 = Add(A=@n35.Cm, B=@n32.Cm)
  n37 = CloudSampleAttribute()
  n38 = Constant()
  n39 = ScalarParameter() {defaultValue=1, parameterName=HighRemap}
  n40 = NamedRerouteUsage()
  n41 = Divide(
   A=Subtract(
    A=MaterialFunctionCall(Input=@n37.NormAltitudeInLayer, Input Low=@n38, Input High=@n39) {materialFunction=&RemapValue_0-1}.Result
    B=ComponentMask(None=@n40) {b=true}
   )
   B=ComponentMask(None=@n40) {a=true}
  )
  n42 = Add(A=Saturate(None=Divide(A=@n34, B=@n36)), B=Saturate(None=OneMinus(None=@n41)))
  n43 = Add(
   A=MaterialFunctionCall!(Vector or UVs=NamedRerouteUsage()) {materialFunction=/Game/Athena/SafeZone/S11/MF_VectorToRadialValueFast}.Result
   B=Multiply(A=Time(), B=ScalarParameter() {group=StormRadialWaves, parameterName=WavesSpeed})
  )
  n44 = Sine(None=Multiply(A=Frac(None=@n43), B=ScalarParameter() {defaultValue=4, group=StormRadialWaves, parameterName=WaveCount}) {constB=10})
  n45 = Power(
   Base=Add(A=Multiply(A=@n44) {constB=0.5}) {constB=0.5}
   Exp=ScalarParameter() {defaultValue=1, group=StormRadialWaves, parameterName=WavesPow}
  )
  n46 = ComponentMask(None=Saturate(None=Add(A=@n42, B=@n45) {constB=0})) {r=true}
  n47 = OneMinus(None=@n46)
  n48 = Multiply(A=@n32.Cm, B=Saturate(None=Add(A=@n33, B=@n47)))
  NamedRerouteDeclaration(None=Divide(A=Subtract(A=@n30, B=@n48), B=@n35.Cm)) {name=Storm Radius, nodeColor=<rgba 0.0547 1 0 1>}
  n49 = MaterialFunctionCall!(WorldScale=ScalarParameter() {defaultValue=2250000, parameterName=worldScale}) {materialFunction=/Game/TimeOfDay/CloudsAndStorm/Materials/MaterialFunctions/MF_WorldRemap}
  n50 = Multiply(A=ComponentMask(None=NamedRerouteUsage()) {g=true, r=true}, B=ScalarParameter() {defaultValue=0.5, parameterName=SphereMaskErosionGain}) {constB=0.2}
  n51 = SphereMask(
   A=ReflectionCapturePassSwitch(Default=Subtract(A=@n49.Result, B=@n50), Reflection=@n49.Result)
   B=Constant()
   Radius=ScalarParameter() {defaultValue=0.2, parameterName=VoidMaskRadius}
   Hardness=ScalarParameter() {defaultValue=0.333, parameterName=VoidMaskHardness}
  )
  NamedRerouteDeclaration(None=Saturate(None=OneMinus(None=@n51))) {name=voidMask, nodeColor=<rgba 0 0.0391 1 1>}
  NamedRerouteDeclaration(None=MaterialFunctionCall(Color_Morning=@n10.RGB, Color_Day=@n11.RGB, Color_Evening=@n12.RGB, Color_Night=@n13.RGB) {materialFunction=&Color_Time_of_Day}.Result) {name=CloudColor, nodeColor=<rgba 0.2422 1 0 1>}
  NamedRerouteDeclaration(None=MaterialFunctionCall(Color_Morning=@n14.RGB, Color_Day=@n15.RGB, Color_Evening=@n16.RGB, Color_Night=@n17.RGB) {materialFunction=&Color_Time_of_Day}.Result) {name=CloudEmissive, nodeColor=<rgba 0.8281 1 0 1>}
  n52 = NamedRerouteUsage()
  n53 = ScalarParameter() {defaultValue=0.001, group=Storm, parameterName=scc}
  n54 = Saturate(None=Add(A=Multiply(A=@n52, B=@n53), B=ScalarParameter() {defaultValue=0.5, group=Storm, parameterName=off}))
  n55 = LinearInterpolate(
   A=ScalarParameter() {defaultValue=128, group=Storm, parameterName=Global Storm Density}
   B=ScalarParameter() {defaultValue=128, group=Storm, parameterName=Global Storm Density Top}
   Alpha=@n54
  )
  n56 = Multiply(
   A=ComponentMask(None=NamedRerouteUsage()) {r=true}
   B=LinearInterpolate(
    A=ScalarParameter() {defaultValue=0.5286, group=StormDFs, parameterName=Noise Intensity}
    B=ScalarParameter() {defaultValue=0.2, group=StormDFs, parameterName=Noise IntensityTop}
    Alpha=@n54
   )
  )
  n57 = Multiply(A=NamedRerouteUsage(), B=VectorParameter() {group=Storm, parameterName=WorldScaleNoise, defaultValue=<rgba 1 1 1 0>}.RGB)
  n58 = LinearInterpolate(
   A=ScalarParameter() {defaultValue=1, group=Storm, parameterName=NoiseGlobalScaleEnd}
   B=ScalarParameter() {defaultValue=1, group=Storm, parameterName=NoiseGlobalScale}
   Alpha=CollectionParameter() {collection=&MPC_SafeZoneClouds, parameterName=SafeZoneScale}
  )
  n59 = Multiply(A=ScalarParameter() {defaultValue=4500000, group=Storm, parameterName=scale}, B=@n58)
  n60 = TextureSample!(
   UVs=Divide(A=@n57, B=MaterialFunctionCall(Km=@n59) {materialFunction=&Km-to-Cm}.Cm)
   Tex
   Level=CollectionParameter() {collection=&MPC_SafeZoneClouds, parameterName=StormDetailVolumeMip}
   Apply View MipBias
  ) {mipValueMode=TMVM_MipLevel, samplerType=SAMPLERTYPE_LinearColor, texture=/Volumetrics/Content/VolumeTextures/Textures/VT_Voronoi_Simple}
  n61 = ScalarParameter() {defaultValue=0.8, group=StormDFs, parameterName=Noise DF Center End}
  n62 = ScalarParameter() {defaultValue=0.8, group=StormDFs, parameterName="Noise DF Center "}
  n63 = CollectionParameter() {collection=&MPC_SafeZoneClouds, parameterName=SafeZoneScale}
  n64 = NamedRerouteDeclaration(None=Subtract(A=@n60.R, B=LinearInterpolate(A=@n61, B=@n62, Alpha=@n63)) {constB=0.5}) {name=StormNoise, nodeColor=<rgba 1 0 0.8437 1>}
  n65 = MaterialFunctionCall(Float2=Normalize(VectorInput=NamedRerouteUsage())) {materialFunction=&BreakOutFloat2Components}
  n66 = AppendVector(A=Multiply(A=@n65.G) {constB=-1}, B=@n65.R)
  n67 = Constant()
  n68 = Multiply(A=@n58, B=ScalarParameter() {defaultValue=512, group=Storm, parameterName=scale2})
  n69 = MaterialFunctionCall!(
   Volume Texture=TextureObject() {samplerType=SAMPLERTYPE_LinearColor, texture=&VT_Voronoi_Simple}
   UVW=Divide(A=@n57, B=MaterialFunctionCall(Km=@n68) {materialFunction=&Km-to-Cm}.Cm)
   Velocity=Multiply(A=AppendVector(A=@n66, B=@n67), B=ScalarParameter() {defaultValue=0.7, group=StormFlowMap, parameterName=Storm Rotation Velocity})
   Override Time=Multiply(A=ScalarParameter() {defaultValue=0.1, group=StormFlowMap, parameterName=T}, B=NamedRerouteUsage())
   Mip Level=CollectionParameter() {collection=&MPC_SafeZoneClouds, parameterName=StormFlowMapsVolumeMip}
  ) {materialFunction=/Game/TimeOfDay/CloudsAndStorm/Materials/MaterialFunctions/Flowmaps_3D}
  n70 = ScalarParameter() {defaultValue=0.3, group=StormDFs, parameterName=Flowmaps DF Center End}
  n71 = ScalarParameter() {defaultValue=0.8, group=StormDFs, parameterName="Flowmaps DF Center "}
  n72 = CollectionParameter() {collection=&MPC_SafeZoneClouds, parameterName=SafeZoneScale}
  n73 = NamedRerouteDeclaration(None=CollectionParameter() {collection=&MPC_SafeZoneClouds, parameterName=CloudStartingBias}) {name=CloudStartingBias, nodeColor=<rgba 0 0.3672 1 1>}
  n74 = Subtract(
   A=NamedRerouteDeclaration(None=ComponentMask(None=@n69.Result RGB) {r=true}) {name=FlowMap, nodeColor=<rgba 1 0 0.1172 1>}
   B=LinearInterpolate(
    A=LinearInterpolate(A=@n70, B=@n71, Alpha=@n72)
    B=ScalarParameter() {defaultValue=1, group=StormDFs, parameterName="Flowmaps DF Center PreStart "}
    Alpha=Saturate(None=@n73)
   )
  ) {constB=0.5}
  n75 = ScalarParameter() {defaultValue=0.5, group=StormDFs, parameterName=Flowmaps Intensity}
  n76 = Add(
   A=ReflectionCapturePassSwitch(Default=Multiply(A=@n56, B=@n64), Reflection=Constant())
   B=NamedRerouteDeclaration(None=Multiply(A=@n74, B=@n75)) {name=Flow, nodeColor=<rgba 0 1 0.5547 1>}
  )
  MaterialExpressionNamedRerouteUsage_4 = NamedRerouteUsage()
  MaterialExpressionComponentMask_15 = ComponentMask(None=NamedRerouteUsage()) {g=true}
  n77 = Custom!(NormAltitude=@MaterialExpressionNamedRerouteUsage_4, Density=@MaterialExpressionComponentMask_15) {
   code="float HeightDensity = Density;\r\n\r\nif(NormAltitude<0.05 || NormAltitude>0.95)\r\n{\r\n   HeightDensity = 0.0;\r\n}\r\n\r\nreturn HeightDensity;"
   inputs [
    {
     inputName=NormAltitude
     input {expression=&self:MaterialExpressionNamedRerouteUsage_4}
    }
    {inputName=Density, input {expression=&self:MaterialExpressionComponentMask_15}}
   ]
  }
  n78 = Multiply(
   A=ScalarParameter() {defaultValue=0.5286, group=StormDFs, parameterName=StormDonut Intensity}
   B=Subtract(A=@n77, B=ScalarParameter() {defaultValue=0.681, group=StormDFs, parameterName=StormDonut DF Center})
  )
  n79 = OneMinus(None=Max(
    A=CollectionParameter() {collection=&MPC_SafeZoneClouds, parameterName=CloudStartingBias}
    B=CollectionParameter() {collection=&MPC_SafeZoneClouds, parameterName=PreventStormShrinking}
   ))
  n80 = ReflectionCapturePassSwitch(
   Default=Multiply(A=CollectionParameter() {collection=&MPC_SafeZoneClouds, parameterName=EnableVolumetricStorm}, B=Saturate(None=@n79))
   Reflection=Constant()
  )
  n81 = Multiply(A=Multiply(A=Divide(A=@n55) {constB=100}, B=Add(A=@n76, B=@n78)), B=@n80)
  NamedRerouteDeclaration(None=@n81) {name=StormEXT, nodeColor=<rgba 1 0 0.8894 1>}
  n82 = CollectionParameter() {collection=&MPC_SafeZoneClouds, parameterName=SafeZoneLoc}
  NamedRerouteDeclaration(None=Subtract(A=WorldPosition().XYZ, B=ComponentMask(None=@n82) {b=true, g=true, r=true})) {name=Safe Zone Local, nodeColor=<rgba 1 0 0.7266 1>}
  n83 = ScalarParameter() {defaultValue=2, parameterName=StormFlowMapCloudInfluence}
  n84 = OneMinus(None=Saturate(None=NamedRerouteUsage()))
  n85 = Multiply(
   A=Multiply(A=@n22, B=NamedRerouteUsage()) {constB=0.0001}
   B=LinearInterpolate(B=NamedRerouteUsage(), Alpha=Multiply(A=@n83, B=@n84)) {constA=1}
  )
  n86 = Multiply(A=@n22, B=ScalarParameter() {defaultValue=0.025, parameterName=reflectionDensity}) {constB=0.0001}
  NamedRerouteDeclaration(None=Multiply(A=ReflectionCapturePassSwitch(Default=@n85, Reflection=@n86), B=ScalarParameter() {defaultValue=1, group=Clouds, parameterName=CloudDensity})) {name=m, nodeColor=<rgba 0 1 0.4844 1>}
  n87 = WorldPosition()
  n88 = MaterialFunctionCall(Km=ScalarParameter() {defaultValue=5, group=Lightning, parameterName=Lighting Pattern Scale}) {materialFunction=&Km-to-Cm}
  n89 = AppendVector(
   A=ComponentMask(None=Divide(A=@n87.XYZ, B=@n88.Cm)) {g=true, r=true}
   B=MaterialFunctionCall(Speed=ScalarParameter() {defaultValue=0.2, group=Lightning, parameterName=Lighting Speed}) {materialFunction=&TimeWithSpeedVariable}.Time
  )
  n90 = TextureSample!(UVs=@n89, Tex, Apply View MipBias) {texture=/Volumetrics/Content/VolumeTextures/Textures/VT_Lightning}
  n91 = ScalarParameter() {defaultValue=3, group=Lightning, parameterName=Lightning Power}
  n92 = Multiply(A=@n90.R, B=@n90.R)
  n93 = ScalarParameter() {group=Lightning, parameterName=Fill Lightning Scatter}
  n94 = ScalarParameter() {defaultValue=2.5, group=Lightning, parameterName=Lightning Mip}
  n95 = Add(
   A=Add(A=Power(Base=@n90.R, Exp=@n91), B=Multiply(A=@n92, B=@n93))
   B=Multiply(
    A=TextureSample!(UVs=@n89, Tex, Level=@n94, Apply View MipBias) {mipValueMode=TMVM_MipLevel, texture=&VT_Lightning}.R
    B=ScalarParameter() {defaultValue=0.05, group=Lightning, parameterName=Lightning Spread}
   )
  ) {constB=0}
  n96 = Power(Base=NamedRerouteUsage(), Exp=ScalarParameter() {defaultValue=1, group=Lightning, parameterName=Lightning Mask Power})
  NamedRerouteDeclaration(None=Multiply(A=Multiply(A=@n95, B=@n96), B=VectorParameter() {group=Lightning, parameterName=Lightning Color, defaultValue=<rgba 0.0997 0.3491 2 0>}.RGB)) {name=Lightning, nodeColor=<rgba 1 0 0.2223 1>}
  NamedRerouteDeclaration(None=MaterialFunctionCall(
    Value_Morning=ScalarParameter() {defaultValue=0.2, group=Scattering and Phase, parameterName=MS Occ Morning}
    Value_Day=ScalarParameter() {defaultValue=0.2, group=Scattering and Phase, parameterName=MS Occ Day}
    Value_Evening=ScalarParameter() {defaultValue=0.2, group=Scattering and Phase, parameterName=MS Occ Evening}
    Value_Night=ScalarParameter() {defaultValue=0.2, group=Scattering and Phase, parameterName=MS Occ Night}
   ) {materialFunction=&Scalar_Time_of_Day}.Result) {name=MultiScatteringOcclusion, nodeColor=<rgba 0.6641 1 0 1>}
  n97 = Add(A=Multiply(A=NamedRerouteUsage(), B=ScalarParameter() {defaultValue=-2, parameterName=StormCloudMaskNoiseStrength}), B=NamedRerouteUsage())
  n98 = Add(
   A=OneMinus(None=Saturate(None=@n97))
   B=Multiply(A=NamedRerouteUsage(), B=ScalarParameter() {defaultValue=2, parameterName=StormCloudMaskStartMultiplier})
  )
  n99 = LinearInterpolate(B=@n98, Alpha=ScalarParameter() {defaultValue=2, parameterName=StormCloudMaskStrength}) {constA=1}
  n100 = CollectionParameter() {collection=&MPC_SafeZoneClouds, parameterName=SafeZoneScale}
  n101 = Add(
   A=SmoothStep(Max=ScalarParameter() {defaultValue=0.2, parameterName=StormScaleFadeClouds}, Value=Saturate(None=@n100)) {constMax=0.2}
   B=ScalarParameter() {parameterName=StormScaleFadeCloudBias}
  )
  NamedRerouteDeclaration(None=Min(A=Multiply(A=@n99, B=@n101))) {name=StormCloudMask, nodeColor=<rgba 0.8281 1 0 1>}
  NamedRerouteDeclaration(None=@n80) {name=Storm Control, nodeColor=<rgba 1 0 0.7734 1>}
  n102 = NamedRerouteUsage()
  n103 = NamedRerouteUsage()
  n104 = TextureSample(
   UVs=MaterialFunctionCall(X=Saturate(None=@n102), Y=OneMinus(None=@n103)) {materialFunction=&MakeFloat2}.Result
   Level=CollectionParameter() {collection=&MPC_SafeZoneClouds, parameterName=StormProfileMip}
  ) {mipValueMode=TMVM_MipLevel, samplerType=SAMPLERTYPE_LinearColor, texture=/Game/TimeOfDay/CloudsAndStorm/Textures/StormProfileLegacy}
  NamedRerouteDeclaration(None=@n104.A) {name=StormShape, nodeColor=<rgba 1 0.141 0.7007 1>}
  n105 = LinearInterpolate(B=@n98, Alpha=ScalarParameter() {defaultValue=4, parameterName=StormCloudMaskColorStrength}) {constA=1}
  NamedRerouteDeclaration(None=Min(A=Multiply(A=@n105, B=@n101))) {name=StormCloudColorMask, nodeColor=<rgba 1 0 0.0703 1>}
  NamedRerouteDeclaration(None=LinearInterpolate(
    A=NamedRerouteUsage()
    B=VectorParameter() {group=HadesClouds, parameterName=Hades_Cloud_Color, defaultValue=<rgba 0.1002 0.1615 0.1001 1>}.RGB
    Alpha=@n29
   )) {name=FinalCloudColor, nodeColor=<rgba 1 0 0.8672 1>}
  n106 = NamedRerouteUsage()
  n107 = Subtract(
   A=MaterialFunctionCall(
    Input=CloudSampleAttribute().NormAltitudeInLayer
    Input Low=ScalarParameter() {defaultValue=-0.02, parameterName=LowRemap}
    Input High=ScalarParameter() {defaultValue=1, parameterName=HighRemap}
   ) {materialFunction=&RemapValue_0-1}.Result
   B=ComponentMask(None=@n106) {b=true}
  )
  n108 = Add(
   A=ComponentMask(None=@n106) {a=true}
   B=Multiply(A=ScalarParameter() {defaultValue=-0.0857, group=StormRadialWaves, parameterName=WavesHeight}, B=@n46)
  )
  NamedRerouteDeclaration(None=Saturate(None=Divide(A=@n107, B=@n108))) {name=Linear Height Profile, nodeColor=<rgba 1 0.1875 0 1>}
  n109 = NamedRerouteUsage()
  n110 = ScalarParameter() {defaultValue=26, parameterName=CloudNoiseFrequency}
  n111 = Time()
  n112 = ScalarParameter() {defaultValue=0.01, desc=0.002, parameterName=CloudNoiseSpeed}
  n113 = MaterialFunctionCall(X=@n112, Y=@n112, Z=Constant()) {materialFunction=&MakeFloat3}
  n114 = TextureSample(UVs=Add(A=Multiply(A=@n109, B=@n110), B=Multiply(A=@n111, B=@n113.Result))) {mipValueMode=TMVM_MipLevel, samplerType=SAMPLERTYPE_LinearColor, texture=/Game/Global/MaterialFunctions/VT_ClothSimWPO}
  n115 = NamedRerouteUsage()
  n116 = ScalarParameter() {defaultValue=-0.2, parameterName=CloudNoiseStrengthLow}
  n117 = ScalarParameter() {defaultValue=0.5, parameterName=CloudNoiseStrengthHigh}
  n118 = NamedRerouteUsage()
  NamedRerouteDeclaration(None=Multiply(A=Multiply(A=@n114.R, B=@n115), B=LinearInterpolate(A=@n116, B=@n117, Alpha=@n118))) {name=CloudNoise, nodeColor=<rgba 0 0.6953 1 1>}
  NamedRerouteDeclaration(None=@n104.RGB) {name=Mapped Storm Profile, nodeColor=<rgba 0.8984 0 1 1>}
  n119 = LinearInterpolate(A=NamedRerouteUsage(), B=VectorParameter() {group=HadesClouds, parameterName=Hades_Cloud_Emissive}.RGB, Alpha=@n29)
  NamedRerouteDeclaration(None=@n119) {name=FinalCloudEmissive, nodeColor=<rgba 1 0 0.8672 1>}
  n120 = NamedRerouteUsage()
  n121 = Saturate(None=MaterialFunctionCall(Input=CloudSampleAttribute().NormAltitudeInLayer, Input Low=NamedRerouteUsage(), Input High=NamedRerouteUsage()) {materialFunction=&RemapValueRange}.Result)
  n122 = TextureSample!(
   UVs=AppendVector(A=ComponentMask(None=@n120) {g=true, r=true}, B=NamedRerouteDeclaration(None=@n121) {name=CloudUVZ, nodeColor=<rgba 0 0.3437 1 1>})
   Tex
   Apply View MipBias
  ) {
   samplerSource=SSM_Wrap_WorldGroupSettings
   samplerType=SAMPLERTYPE_LinearColor
   texture=/Game/TimeOfDay/CloudsAndStorm/Textures/VolumeCloudSDF_v04_Volume
  }
  n123 = Add(A=NamedRerouteUsage())
  n124 = ScalarParameter() {defaultValue=0.4, parameterName=CloudMaskStrength}
  n125 = MaterialFunctionCall(
   Input=Multiply(A=@n122.R, B=LinearInterpolate(B=@n123, Alpha=@n124) {constA=1})
   Input Low=NamedRerouteUsage()
   Input High=ScalarParameter() {defaultValue=-0.1667, parameterName=SDF_Min}
  ) {materialFunction=&RemapValue_0-1}
  NamedRerouteDeclaration(None=@n125.Result) {name=RawCloudSDF, nodeColor=<rgba 1 0.5391 0 1>}
  n126 = NamedRerouteUsage()
  n127 = ScalarParameter() {defaultValue=0.6, parameterName=CloudConservativeExpand}
  n128 = Sign(None=Multiply(A=Subtract(A=@n126, B=@n127)) {constB=-1})
  NamedRerouteDeclaration(None=Multiply(A=Saturate(None=@n128), B=NamedRerouteUsage())) {name=ConservativeClouds, nodeColor=<rgba 0 0.0156 1 1>}
  n129 = NamedRerouteUsage()
  n130 = Saturate(None=NamedRerouteUsage())
  n131 = Multiply(A=Ceil(None=@n130), B=Ceil(None=OneMinus(None=@n130)))
  NamedRerouteDeclaration(None=MaterialFunctionCall(X=Max(A=@n129, B=@n131), Y=NamedRerouteUsage()) {materialFunction=&MakeFloat4}.Result) {name=ConservativeVector, nodeColor=<rgba 0 0.8594 1 1>}
  n132 = NamedRerouteUsage()
  n133 = Multiply(A=Add(A=WorldPosition().XYZ, B=Multiply(A=@n132) {constB=0.5}), B=Divide(A=Constant() {r=1}, B=@n132))
  n134 = Multiply(A=NamedRerouteUsage(), B=NamedRerouteUsage())
  n135 = Multiply(A=ComponentMask(None=@n134) {r=true}) {constB=0.6667}
  NamedRerouteDeclaration(None=Subtract(
    A=@n133
    B=MaterialFunctionCall(X=@n134, Y=@n134, Z=@n135) {materialFunction=&MakeFloat3}.Result
   )) {name=worldUV, nodeColor=<rgba 0 0.7188 1 1>}
  NamedRerouteDeclaration(None=ScalarParameter() {defaultValue=2250000, parameterName=worldScale}) {name=Scale, nodeColor=<rgba 1 0 0.8203 1>}
  NamedRerouteDeclaration(None=LinearInterpolate(A=NamedRerouteUsage(), B=ScalarParameter() {defaultValue=1, group=HadesClouds, parameterName=Hades_Cloud_Occlusion}, Alpha=@n29)) {name=CloudOcclusion, nodeColor=<rgba 1 0 0.8672 1>}
  n136 = CollectionParameter() {collection=&MPC_SafeZoneClouds, parameterName=StormThicknessKm-End}
  n137 = CollectionParameter() {collection=&MPC_SafeZoneClouds, parameterName=StormRadiusKm-End}
  n138 = CollectionParameter() {collection=&MPC_SafeZoneClouds, parameterName=StormAltitude-End}
  n139 = CollectionParameter() {collection=&MPC_SafeZoneClouds, parameterName=StormHeight-End}
  n140 = CollectionParameter() {collection=&MPC_SafeZoneClouds, parameterName=StormThicknessKm}
  n141 = Saturate(None=CollectionParameter() {collection=&MPC_SafeZoneClouds, parameterName=CloudStartingBias})
  n142 = Add(
   A=CollectionParameter() {collection=&MPC_SafeZoneClouds, parameterName=StormRadiusKm}
   B=Multiply(A=ScalarParameter() {defaultValue=3, parameterName=StormRadius_PreStartOffset}, B=Multiply(A=@n141, B=@n141))
  )
  n143 = CollectionParameter() {collection=&MPC_SafeZoneClouds, parameterName=StormAltitude}
  n144 = CollectionParameter() {collection=&MPC_SafeZoneClouds, parameterName=StormHeight}
  NamedRerouteDeclaration(None=LinearInterpolate(
    A=MaterialFunctionCall(X=@n136, Y=@n137, Z=@n138, A=@n139) {materialFunction=&MakeFloat4}.Result
    B=MaterialFunctionCall(X=@n140, Y=@n142, Z=@n143, A=@n144) {materialFunction=&MakeFloat4}.Result
    Alpha=CollectionParameter() {collection=&MPC_SafeZoneClouds, parameterName=SafeZoneScale}
   )) {name=Storm Collection Params, nodeColor=<rgba 0.4708 0 1 1>}
  NamedRerouteDeclaration(None=LinearInterpolate(A=NamedRerouteUsage(), Alpha=@n29)) {name=FinalVoidMask, nodeColor=<rgba 1 0 0.8672 1>}
  NamedRerouteDeclaration(None=@n54) {name=StormHeightMask, nodeColor=<rgba 1 0 0.2578 1>}
  n145 = WorldPosition()
  n146 = MaterialFunctionCall(
   R=ScalarParameter() {defaultValue=-509338.6875, group=Kelp_Clouds, parameterName=Kelp_VolumetricX_Position}
   G=ScalarParameter() {defaultValue=-766436.25, group=Kelp_Clouds, parameterName=Kelp_VolumetricY_Position}
   B=ScalarParameter() {defaultValue=202012.8281, group=Kelp_Clouds, parameterName=Kelp_VolumetricZ_Position}
  ) {materialFunction=&AppendMany}
  n147 = ScalarParameter() {defaultValue=1051773.125, group=Kelp_Clouds, parameterName=Kelp_Volumetric_Radius}
  n148 = NamedRerouteUsage()
  n149 = Multiply(
   A=TextureSample(UVs=ComponentMask(None=@n148) {g=true, r=true}) {samplerSource=SSM_Wrap_WorldGroupSettings, texture=/Game/TimeOfDay/TODM_Sequence/Materials/Sky/Clouds/T_Clouds_02_M}.RGB
   B=ScalarParameter() {defaultValue=1, group=Kelp_Clouds, parameterName=Cloud_Radius_Noise_Multi}
  )
  n150 = Subtract(
   A=Multiply(
    A=SphereMask(A=@n145.XYZ, B=@n146.RGB, Radius=@n147) {hardnessPercent=0.5}
    B=ScalarParameter() {defaultValue=2.729, group=Kelp_Clouds, parameterName=CloudSize_Multi}
   )
   B=@n149
  )
  NamedRerouteDeclaration(None=LinearInterpolate(B=Saturate(None=@n150), Alpha=ScalarParameter() {group=Kelp_Clouds, parameterName=Kelp_Enabled_Volumetric}) {constA=1}) {name=Kelp_Cloud_Mask, nodeColor=<rgba 1 0 0.3984 1>}
  NamedRerouteDeclaration(None=LinearInterpolate(
    B=MaterialFunctionCall() {materialFunction=/ClydeCosmetics/Environments/KelpLinen/MF_KelpLinen_LocationFade}.Output
    Alpha=ScalarParameter() {group=Kelp_Clouds, parameterName=Kelp_Enabled_Volumetric}
   ) {constA=1}) {name=Kelp_Play_Area, nodeColor=<rgba 0 0.625 1 1>}
  n151 = VectorParameter() {group=Kelp_Clouds, parameterName=Kelp_Glow_Position, defaultValue=<rgba -670680.75 -844955.8125 36048.2383 1>}
  n152 = ScalarParameter() {defaultValue=523742.5312, desc=Radius of glow area to mask clouds around kelp, group=Kelp_Clouds, parameterName=Kelp_Glow_Mask_Radius}
  n153 = MaterialFunctionCall() {materialFunction=/Engine/Functions/Engine_MaterialFunctions02/Utility/FOV}
  n154 = SphereMask(
   A=ScreenPosition().PixelPosition
   B=MaterialFunctionCall!(Center Point World Position=@n151.RGB) {materialFunction=/Game/Packages/Fortnite_UniqueMaterials_1/Effects/WorldAlignedScreenCoordinates}.Result
   Radius=Multiply(A=@n152, B=Divide(B=@n153.FOV angle in degrees) {constA=90})
  ) {hardnessPercent=0.1}
  NamedRerouteDeclaration(None=LinearInterpolate(B=Power(Base=@n154), Alpha=ScalarParameter() {group=Kelp_Clouds, parameterName=Kelp_Enabled_Volumetric})) {name=Kelp_Glow_Mask, nodeColor=<rgba 0.9688 1 0 1>}
  n155 = WorldPosition()
  n156 = VectorParameter() {
   desc="Offset for the glow position to help the fresnel on the volumetrics look more accurate.  Adds to Kelp Glow Position"
   group=Kelp_Clouds
   parameterName=Kelp_Glow_Offset
   defaultValue=<rgba 0 0 0 1>
  }
  n157 = Add(A=@n156.RGB, B=@n151.RGB)
  n158 = ScalarParameter() {defaultValue=523742.5312, desc="Radius of cloud fresnel glow ", group=Kelp_Clouds, parameterName=Kelp_Glow_Radius}
  n159 = LinearInterpolate(
   A=@n158
   B=Add(A=@n158, B=ScalarParameter() {defaultValue=82742.5312, desc="Radius of cloud fresnel glow ", group=Kelp_Clouds, parameterName=Music_Offset})
   Alpha=NamedRerouteUsage()
  )
  n160 = ScalarParameter() {defaultValue=0.2, desc="Radius of cloud fresnel glow ", group=Kelp_Clouds, parameterName=Kelp_Glow_Radius_Softness}
  n161 = Power(
   Base=DotProduct(A=CameraVectorWS(), B=SphereMask(A=@n155.XYZ, B=@n157, Radius=@n159, Hardness=@n160) {hardnessPercent=0.2})
   Exp=ScalarParameter() {defaultValue=2, desc="Radius of cloud fresnel glow ", group=Kelp_Clouds, parameterName=Kelp_Glow_Pow}
  )
  n162 = ScalarParameter() {defaultValue=1, group=Kelp_Clouds, parameterName=Kelp_Glow_Multi_Volumetric}
  NamedRerouteDeclaration(None=LinearInterpolate(B=Multiply(A=@n161, B=@n162), Alpha=ScalarParameter() {group=Kelp_Clouds, parameterName=Kelp_Backdrop_enabled})) {name=Kelp_Backdrop_Glow, nodeColor=<rgba 0 1 0.5078 1>}
  NamedRerouteDeclaration(None=MaterialFunctionCall!(In) {materialFunction=/ClydeCosmetics/Environments/KelpLinen/MF_Kelp_AudioReactive}.Beat_Tracker) {name=Music beat, nodeColor=<rgba 1 0.1406 0 1>}
  n163 = Constant() {desc=Default Radius - Don't Touch unless fade distance for entire zone is changing, r=131922.5}
  n164 = ScalarParameter() {
   desc="Use this to make a element fade faster or slower than other objects when the camera moves thru the zone.  Tweak carefully.  "
   group=Kelp_Location
   parameterName=StormBackFade
  }
  n165 = Multiply(
   A=SphereMask(
    A=CameraPositionWS()
    B=Constant3Vector() {constant=<rgba -27852 -50384 -63335.4922 1>}
    Radius=Add(A=@n163, B=@n164)
    Hardness=Constant() {desc=Default for old values0.5, r=0.85}
   )
   B=Constant() {r=2}
  )
  NamedRerouteDeclaration(None=LinearInterpolate(B=Saturate(None=@n165), Alpha=ScalarParameter() {group=Kelp_Clouds, parameterName=Kelp_Enabled_Volumetric}) {constA=1}) {name=Cutout_Play_Area, nodeColor=<rgba 0 1 0.7422 1>}
  NamedRerouteDeclaration(None=ScalarParameter() {defaultValue=-0.1, parameterName=CloudHeightRemapLow}) {name=LowRemap, nodeColor=<rgba 0.4297 0 1 1>}
  NamedRerouteDeclaration(None=ScalarParameter() {defaultValue=1.5, parameterName=CloudHeightRemapHigh}) {name=HighRemap, nodeColor=<rgba 0.9688 0 1 1>}
  NamedRerouteDeclaration(None=@n69.UVW_A) {name=FlowUVs, nodeColor=<rgba 0 0.5312 1 1>}
  NamedRerouteDeclaration(None=ScalarParameter() {defaultValue=0.001, desc=0.002, parameterName=BaseCloudSpeed}) {name=CloudSpeed, nodeColor=<rgba 1 0.4453 0 1>}
  n166 = VectorParameter() {group=Kelp_Clouds, parameterName=Kelp_Storm_Location_Position, defaultValue=<rgba -670680.75 -710886.375 108055.1875 1>}
  n167 = ScalarParameter() {defaultValue=692407.125, desc=Radius of glow area to mask clouds around kelp, group=Kelp_Clouds, parameterName=Kelp_Storm_Mask_Radius}
  n168 = Divide(B=MaterialFunctionCall() {materialFunction=&FOV}.FOV angle in degrees) {constA=90}
  n169 = Subtract(
   A=SphereMask(
    A=ScreenPosition().PixelPosition
    B=MaterialFunctionCall(Center Point World Position=@n166.RGB) {materialFunction=&WorldAlignedScreenCoordinates}.Result
    Radius=Multiply(A=@n167, B=@n168)
    Hardness=ScalarParameter() {defaultValue=0.359, parameterName=Screen_Hardness}
   ) {hardnessPercent=1}
   B=NamedRerouteUsage()
  )
  NamedRerouteDeclaration(None=LinearInterpolate(B=Saturate(None=@n169), Alpha=ScalarParameter() {group=Kelp_Clouds, parameterName=Kelp_Enabled_Volumetric})) {name=Kelp_Storm_Cutout, nodeColor=<rgba 1 0.375 0 1>}
  NamedRerouteDeclaration(None=@n149) {name=storm_Kelp_mask, nodeColor=<rgba 0.0547 0 1 1>}
  n170 = Constant() {desc=Default Radius - Don't Touch unless fade distance for entire zone is changing, r=131922.5}
  n171 = ScalarParameter() {defaultValue=4600, group=Kelp_Location, parameterName=Storm_Front_Fade}
  n172 = Multiply(
   A=SphereMask(
    A=CameraPositionWS()
    B=Constant3Vector() {constant=<rgba -27852 -50384 -63335.4922 1>}
    Radius=Add(A=@n170, B=@n171)
    Hardness=Constant() {desc=Default for old values0.5, r=0.85}
   )
   B=Constant() {r=2}
  )
  NamedRerouteDeclaration(None=LinearInterpolate(B=Saturate(None=@n172), Alpha=ScalarParameter() {group=Kelp_Clouds, parameterName=Kelp_Enabled_Volumetric}) {constA=1}) {name=Storm_Play_ARea, nodeColor=<rgba 0 1 0.9531 1>}
  NamedRerouteDeclaration(None=Saturate(None=CollectionParameter() {collection=&MPC_SafeZoneClouds, parameterName=EnableVolumetricStorm})) {name=EnableVolumetricStorm, nodeColor=<rgba 0 0.1328 1 1>}
  n173 = LinearInterpolate(
   A=VectorParameter() {group=Storm, parameterName=AlbedoStormBottom, defaultValue=<rgba 0.2602 0.1591 0.2656 1>}.RGB
   B=VectorParameter() {group=Storm, parameterName=AlbedoStormTop, defaultValue=<rgba 0 0.0938 0.0649 1>}.RGB
   Alpha=NamedRerouteUsage()
  )
  n174 = NamedRerouteUsage()
  n175 = ReflectionCapturePassSwitch(
   Default=LinearInterpolate(A=NamedRerouteDeclaration(None=@n173) {name=StormAlbedo, nodeColor=<rgba 0.7344 0 1 1>}, B=@n174, Alpha=NamedRerouteUsage())
   Reflection=@n174
  )
  n176 = VectorParameter() {parameterName=AlbedoStormMid, defaultValue=<rgba 0.0417 0.0417 0.0417 1>}
  n177 = NamedRerouteUsage()
  n178 = NamedRerouteUsage()
  NamedRerouteDeclaration(None=Multiply(
    A=@n7.RGB
    B=MaterialFunctionCall(A=@n175, B=@n176.RGB, C=@n177, Alpha=@n178) {materialFunction=&Lerp_3Color}.Result
   )) {name=CloudsAlbedo, nodeColor=<rgba 0 0.625 1 1>}
  NamedRerouteDeclaration(None=@n133) {name=worldUVStormStatic, nodeColor=<rgba 0 0.9062 1 1>}
  n179 = Constant3Vector() {constant=<rgba 36194.6641 -67295.8047 0 0>}
  NamedRerouteDeclaration(None=SphereMask(A=ComponentMask(None=@n179) {g=true, r=true}, B=WorldPosition().XY, Radius=ScalarParameter() {defaultValue=200000, parameterName=Radius}) {attenuationRadius=10000, hardnessPercent=50}) {name=HadesSphereMask, nodeColor=<rgba 1 0.4688 0 1>}
  n180 = NamedRerouteUsage()
  n181 = Multiply(A=NamedRerouteUsage(), B=NamedRerouteUsage())
  n182 = Multiply(A=NamedRerouteUsage(), B=NamedRerouteUsage())
  n183 = CloudSampleAttribute()
  n184 = Constant() {r=1.1}
  n185 = OneMinus(None=Saturate(None=Multiply(A=@n183.NormAltitudeInLayer, B=@n184)))
  n186 = Multiply(A=@n185, B=@n185)
  NamedRerouteDeclaration(None=Add(A=Multiply(A=@n180, B=@n181), B=Multiply(A=@n182, B=@n186))) {name=CloudsEmissive, nodeColor=<rgba 0.8516 1 0 1>}
  n187 = Power(Base=NamedRerouteUsage(), Exp=ScalarParameter() {defaultValue=1.85, parameterName=AltitudeCurveExponent})
  NamedRerouteDeclaration(None=LinearInterpolate(
    A=ScalarParameter() {defaultValue=0.05, parameterName=DensityLow}
    B=ScalarParameter() {defaultValue=0.05, parameterName=DensityHigh}
    Alpha=NamedRerouteDeclaration(None=@n187) {name=AltitudeMap, nodeColor=<rgba 0.9453 1 0 1>}
   )) {name=AltitudeDensity, nodeColor=<rgba 0.2891 0 1 1>}
  n188 = NamedRerouteUsage()
  n189 = TextureSampleParameter2D(UVs=Multiply(A=ComponentMask(None=@n188) {g=true, r=true}, B=ScalarParameter() {defaultValue=0.3, parameterName=CloudMaskTilie})) {parameterName=CloudMask, samplerType=SAMPLERTYPE_Grayscale, texture=&LowResBlurredNoise}
  n190 = ScalarParameter() {defaultValue=1.5, parameterName=CloudMaskExponent}
  NamedRerouteDeclaration(None=Saturate(None=Power(Base=@n189.R, Exp=@n190))) {name=CloudMask, nodeColor=<rgba 0 1 0.7656 1>}
  n191 = ComponentMask(None=VolumetricAdvancedMaterialInput().ConservativeDensity as Float4) {g=true}
  n192 = ScalarParameter() {
   defaultValue=0.5
   desc="In SDF space. how far away from the cloud surface, the colour starts to belnd from white to storm"
   parameterName=CloudStormColorBlendStartDistance
  }
  n193 = Add(A=@n192, B=ScalarParameter() {defaultValue=0.3, parameterName=CloudStormColorBlendWidth})
  NamedRerouteDeclaration(None=Power(Base=MaterialFunctionCall(Input=@n191, Input Low=@n192, Input High=@n193) {materialFunction=&RemapValue_0-1}.Result) {constExponent=0.4}) {name=StormCloudBlend, nodeColor=<rgba 0 0.2266 1 1>}
  NamedRerouteUsage()
  NamedRerouteUsage()
  NamedRerouteUsage()
  NamedRerouteUsage()
  NamedRerouteUsage()
  NamedRerouteUsage()
  n194 = NamedRerouteUsage()
  n195 = Clamp(None=@n194, Max=ScalarParameter() {parameterName=Max Cloud Reflection Brightness})
  RayTracingQualitySwitch(Normal=ReflectionCapturePassSwitch(Default=@n194, Reflection=@n195), RayTraced=@n195)
  ReflectionCapturePassSwitch(Default=@n1, Reflection=Multiply(A=@n1, B=ScalarParameter() {defaultValue=0.6, parameterName=ReflectionAlbedoBrightness}))
  n196 = Subtract(
   A=Multiply(A=@n125.Result, B=NamedRerouteUsage())
   B=Multiply(A=NamedRerouteUsage(), B=ScalarParameter() {defaultValue=1, group=Kelp_Clouds, parameterName=Kelp_Glow_Mask_Str})
  )
  n197 = NamedRerouteUsage()
  Reroute(None=Saturate(None=Multiply(A=@n196, B=@n197))) {desc=connect for kelp linen}
  n198 = VectorParameter() {group=Kelp_Clouds, parameterName=Kelp_Backdrop_Glow_Volumetric, defaultValue=<rgba 10 10 10 1>}
  n199 = ScalarParameter() {defaultValue=0.1, desc=Used to reduce the glow on the emissive for kelp, group=Kelp_Clouds, parameterName=Kelp_Backdrop_Glow_Multi}
  n200 = Multiply(A=NamedRerouteUsage(), B=NamedRerouteUsage())
  n201 = OneMinus(None=Max(
    A=CollectionParameter() {collection=&MPC_SafeZoneClouds, parameterName=CloudStartingBias}
    B=CollectionParameter() {collection=&MPC_SafeZoneClouds, parameterName=PreventStormShrinking}
   ))
  n202 = Saturate(None=Multiply(A=CollectionParameter() {collection=&MPC_SafeZoneClouds, parameterName=EnableVolumetricStorm}, B=Saturate(None=@n201)))
  Reroute(None=LinearInterpolate(A=@n119, B=Multiply(A=@n198.RGB, B=@n199), Alpha=LinearInterpolate(A=@n200, Alpha=@n202) {constB=0})) {desc=reconnect for kelp}
  n203 = Multiply(A=Multiply(A=NamedRerouteUsage(), B=ScalarParameter() {defaultValue=0.14, parameterName=Kelp_Storm_Cutout}), B=NamedRerouteUsage())
  n204 = NamedRerouteUsage()
  Reroute(None=Subtract(A=@n81, B=Multiply(A=@n203, B=@n204))) {desc=reconnect for kelp linen}
  ScalarParameter() {defaultValue=0.1, desc=0.1, parameterName=CloudNoiseGlobalStrength}
  ScalarParameter() {defaultValue=0.5, parameterName=SDF_Max}
  ScalarParameter() {defaultValue=0.5, parameterName=AO}
  VolumetricAdvancedMaterialOutput!(
   PhaseG=ScalarParameter() {defaultValue=-0.1, parameterName=PhaseG}
   PhaseG2=ScalarParameter() {defaultValue=0.7, parameterName=PhaseG2}
   PhaseBlend=ScalarParameter() {defaultValue=0.5, parameterName=PhaseBlend}
   MultiScatteringContribution=ScalarParameter() {defaultValue=1, parameterName=MultiScatteringContribution}
   MultiScatteringOcclusion=NamedRerouteUsage()
   MultiScatteringEccentricity=ScalarParameter() {defaultValue=0.5, parameterName=MultiScatteringEccentricity}
   ConservativeDensity=NamedRerouteUsage()
  ) {multiScatteringApproximationOctaveCount=1}
  n205 = ScalarParameter() {defaultValue=0.7, parameterName=CloudConservativeExpandSkip}
  n206 = Multiply(A=Multiply(A=Subtract(A=@n126, B=@n205)) {constB=-1}) {constB=100}
  VolumetricCloudEmptySpaceSkippingOutput!(ContainsMatter=Saturate(None=Sign(None=@n206)))
 ]
}
