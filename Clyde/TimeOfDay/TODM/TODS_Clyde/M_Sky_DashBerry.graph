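# Unlit sky-dome material for the Clyde DashBerry time of day, layering a location-specific "Kelp" galaxy and glow over the Fortnite sky.
# Atmosphere, night gradient, stars, Helios sun/moon and clouds blend through named reroutes, with storm tinting and quality switches.
# Kelp layers add swirl-warped galaxy, tiled stars and audio-beat backdrop glow, masked by play-area and camera/player-position fades.
# Reflection captures get a separately masked, tinted, time-of-day-desaturated result.
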Material /Clyde/TimeOfDay/TODM/TODS_Clyde/M_Sky_DashBerry
graph {
 properties {bIsSky=true, pixelDepthOffsetMode=PDOM_Legacy, shadingModel=MSM_Unlit, usedShadingModels=Unlit}
 nodes [
  n1 = Time()
  n2 = ComponentMask(None=OneMinus(None=Frac(None=@n1))) {r=true}
  Clamp(None=Multiply(A=Power(Base=@n2) {constExponent=6}) {constB=2})
  Constant()
  Constant() {r=1}
  Constant()
  n3 = NamedRerouteUsage()
  LinearInterpolate(A=ScalarParameter() {defaultValue=0.8, group=Beat, parameterName=Glow_Beat_Value}, Alpha=@n3) {constA=1}
  n4 = VectorParameter() {group=Kelp_Clouds, parameterName=Kelp_Backdrop_GlowColor, sortPriority=31, defaultValue=<rgba 10 10 10 1>}
  n5 = Multiply(A=@n4.RGB, B=LinearInterpolate(Alpha=Clamp(None=@n3)) {constA=1, constB=1.5})
  n6 = Constant3Vector() {constant=<rgba 0 2.9994 5 1>}
  LinearInterpolate(A=@n5, B=LinearInterpolate(A=@n6, B=@n4.RGB, Alpha=@n3), Alpha=ScalarParameter() {parameterName=Hue_Shift?})
  n7 = MaterialFunctionCall() {materialFunction=&/MF_Helios_Skybox_Clouds_Clyde}
  n8 = Subtract(A=@n7.CloudAlpha, B=Multiply(A=NamedRerouteUsage(), B=ScalarParameter() {defaultValue=1, parameterName=Backglow_CloudMaskMulti}))
  n9 = NamedRerouteUsage()
  MaterialFunctionCall(A=@n7.CloudAlpha, B=Constant(), C=Saturate(None=Multiply(A=@n8, B=@n9)), Alpha=NamedRerouteUsage()) {materialFunction=&Lerp_3Color}
  n10 = NamedRerouteUsage()
  Max(A=@n10, B=ScalarParameter() {parameterName=Hue_Shift?})
  Multiply()
  n11 = CameraVectorWS()
  n12 = MaterialFunctionCall(
   UVs=TextureCoordinate()
   CenterPosition=AppendVector(
    A=ScalarParameter() {defaultValue=0.88, group=Kelp_Clouds, parameterName=Cloud_Radius_X_Position}
    B=ScalarParameter() {defaultValue=0.9758, group=Kelp_Clouds, parameterName=CloudRadius_Y_Position}
   )
   Radius=ScalarParameter() {defaultValue=0.6757, group=Kelp_Clouds, parameterName=CloudRadius_BackdropGlow_Size}
  ) {materialFunction=&RadialGradientExponential}
  n13 = Multiply(
   A=Power(Base=DotProduct(A=@n11, B=@n12.RadialGradientExponential))
   B=ScalarParameter() {group=Kelp_Glow, parameterName=Cloud_Backdrop_Mutli}
  )
  Multiply(B=@n13)
  Multiply(
   A=MaterialFunctionCall(Hue Shift Percentage=ScalarParameter() {defaultValue=0.7717, parameterName=Hue_Shift_Percentage}, Texture=Multiply(A=@n4.RGB, B=@n6)) {materialFunction=&HueShift}.Result
   B=ScalarParameter() {defaultValue=1.5, parameterName=Hue_Shift_Multi}
  )
  n14 = Subtract(A=WorldPosition().XYZ, B=CameraPositionWS())
  n15 = Saturate(None=DotProduct(A=SkyAtmosphereLightDirection(), B=Normalize(VectorInput=@n14)))
  n16 = LinearInterpolate(A=Constant() {r=0.7}, Alpha=Power(Base=ConstantBiasScale(None=@n15), Exp=Constant() {r=3}))
  n17 = VectorParameter() {group=Night Sky, parameterName=FakeSkyLight, defaultValue=<rgba 0.0144 0.1912 0.3073 1>}
  n18 = Add(
   A=StaticSwitchParameter(True=Multiply(A=@n16, B=@n17.RGB), False=Constant()) {parameterName=UseFakeSkylight}
   B=SkyAtmosphereViewLuminance!(WorldDirection)
  )
  n19 = TextureCoordinate() {uTiling=0.5, vTiling=0.5}
  n20 = AppendVector(A=Constant() {r=0.25}, B=Constant() {r=0.25})
  n21 = LinearInterpolate(Alpha=ScalarParameter() {defaultValue=0.25, group=Night Sky, parameterName=Nightsky Horizon Height}) {constA=0.25, constB=0.2}
  n22 = Constant()
  n23 = CurveAtlasRowParameter(CurveTime=NamedRerouteDeclaration(None=SphereMask(A=@n19, B=@n20, Radius=@n21, Hardness=@n22)) {name=HorizonLine, nodeColor=<rgba 0.875 1 0 1>}) {
   atlas=/Game/TimeOfDay/TODM_Sequence/Materials/Sky/MaterialFunctions/CA_Athena_TOD
   curve=/Game/TimeOfDay/TODM_Sequence/Materials/Sky/MaterialFunctions/NightSkyHorizonCurve
   defaultValue=1
   parameterName=Sky Horizon Curve
  }
  n24 = LinearInterpolate(
   A=LinearInterpolate(
    A=VectorParameter() {group=Night Sky, parameterName=nightsky horizon color, defaultValue=<rgba 15 0 0.4283 1>}.RGB
    B=VectorParameter() {group=Night Sky, parameterName=nightsky color, defaultValue=<rgba 0.405 0.45 0.4458 1>}.RGB
    Alpha=Saturate(None=@n23.G)
   )
   B=SkyAtmosphereViewLuminance()
   Alpha=ScalarParameter() {defaultValue=1, group=Night Sky, parameterName=NightSkyViewLuminanceFactor}
  ) {constA=1}
  n25 = LinearInterpolate(
   A=@n24
   B=Add(A=@n24, B=NamedRerouteUsage())
   Alpha=ScalarParameter() {defaultValue=1, group=Night Sky, parameterName=Nightsky Galaxy Strength}
  )
  n26 = ScalarParameter() {defaultValue=1, group=Night Sky, parameterName=Night Sky blend}
  NamedRerouteDeclaration(None=Add(A=LinearInterpolate(A=@n18, B=@n25, Alpha=@n26), B=NamedRerouteUsage())) {name=Horizon Normal, nodeColor=<rgba 1 0 0.8437 1>}
  n27 = SkyAtmosphereDistantLightScatteredLuminance()
  n28 = Saturate(None=Multiply(A=NamedRerouteUsage()) {constB=6})
  n29 = ScalarParameter() {defaultValue=19000000, group=HighAltitudeClouds, parameterName=HA_Cloud_VirtualplaneScale}
  n30 = Panner(
   Coordinate=MaterialFunctionCall(
    Plane Normal Axis=Constant3Vector() {constant=<rgba 0 0 -1 0>}
    Plane X Axis=Constant3Vector() {constant=<rgba 1 0 0 0>}
    Plane Center=VectorParameter() {parameterName=HA_Cloud_Height, defaultValue=<rgba 0 0 1200000 1>}.RGB
    UV Scale=AppendVector(A=@n29, B=@n29)
   ) {materialFunction=&VirtualPlaneCoordinates}.UVs
   Speed=AppendVector(
    A=ScalarParameter() {defaultValue=-0.0001, group=HighAltitudeClouds, parameterName=HA_Cloud_Speed_X}
    B=ScalarParameter() {defaultValue=0.0005, group=HighAltitudeClouds, parameterName=HA_Cloud_Speed_Y}
   )
  )
  n31 = TextureSampleParameter2D(UVs=Multiply(A=@n30, B=ScalarParameter() {defaultValue=2.1989, group=HighAltitudeClouds, parameterName=HighAlititudeClouds_Tiling})) {
   group=HighAltitudeClouds
   parameterName=HighAlittudeCloudTexture
   samplerSource=SSM_Wrap_WorldGroupSettings
   samplerType=SAMPLERTYPE_LinearColor
   texture=/Game/TimeOfDay/TODM_Sequence/Materials/Sky/Textures/Clouds_m
  }
  n32 = AppendVector(
   A=ScalarParameter() {defaultValue=0, group=HighAltitudeClouds, parameterName=HA_Cloud_Speed2_X}
   B=ScalarParameter() {defaultValue=0.0001, group=HighAltitudeClouds, parameterName=HA_Cloud_Speed2_Y}
  )
  n33 = TextureSampleParameter2D(UVs=Multiply(
    A=Panner(Coordinate=@n30, Speed=@n32)
    B=ScalarParameter() {defaultValue=3, group=HighAltitudeClouds, parameterName=HighAlititudeClouds_Tiling_3}
   )) {
   group=HighAltitudeClouds
   parameterName=HighAlittudeCloudTexture
   samplerSource=SSM_Wrap_WorldGroupSettings
   samplerType=SAMPLERTYPE_LinearColor
   texture=&Clouds_m
  }
  n34 = Power(Base=Multiply(A=@n31.G, B=@n33.B)) {constExponent=1.8}
  n35 = Saturate(None=Multiply(A=Multiply(A=@n28, B=@n34), B=ScalarParameter() {defaultValue=0.5, group=HighAltitudeClouds, parameterName=High Alt Cloud Opacity}))
  NamedRerouteDeclaration(None=Multiply(A=LinearInterpolate(A=@n27) {constAlpha=0.05, constB=0.5}, B=@n35)) {name=High Alt Cloud Color, nodeColor=<rgba 1 0 0.6562 1>}
  NamedRerouteDeclaration(None=@n35) {name=High Alt Cloud Mask, nodeColor=<rgba 1 0.3047 0 1>}
  n36 = TextureSample(UVs=Panner() {speedX=0.0025, speedY=-0.004}) {samplerSource=SSM_Wrap_WorldGroupSettings, samplerType=SAMPLERTYPE_Normal, texture=/Clyde/Environments/KelpLinen/T_GaussianNoise_Normal}
  n37 = Constant()
  n38 = ComponentMask(None=LinearInterpolate(
    A=Constant3Vector() {constant=<rgba 0 0 1 0>}
    B=MaterialFunctionCall!(Default=@n36.RGB, Low/Mobile=@n37) {materialFunction=/Game/Effects/Fort_Effects/Materials/_Functions/MF_LowMobile_PathingSwitch}.Result
    Alpha=Constant() {r=0.1}
   )) {g=true, r=true}
  n39 = Multiply(A=TextureCoordinate(), B=ScalarParameter() {defaultValue=1, group=Spirals, parameterName=Cloud-TileMaster}) {constB=0.5}
  n40 = Constant() {r=1}
  n41 = ScalarParameter() {defaultValue=0.0008, group=Spirals, parameterName=SwirlPanSpeed}
  n42 = ScalarParameter() {defaultValue=4, group=Spirals, parameterName=Cloud-SpeedMaster}
  n43 = MaterialFunctionCall!(
   SineAnimOffset
   Flow Direction=Constant2Vector() {g=0.2, r=-1}
   Time offset
   EdgeMaskSize
   UVS * size=Multiply(A=Subtract(A=@n39, B=@n40), B=ScalarParameter() {defaultValue=5, group=Spirals, parameterName=SwirlTile}) {constB=0.5}
   SineAnimSpeed=ScalarParameter() {defaultValue=0.0005, group=Spirals, parameterName=SwirlCurlSpeed}
   Grid Noise Tile Rep
   Rotation amount modulator=Add(A=Multiply() {constB=0.33}, B=ScalarParameter() {defaultValue=0.05, group=Spirals, parameterName=SwirlAmount})
   un frac Time=MaterialFunctionCall(Speed=Multiply(A=@n41, B=@n42)) {materialFunction=&TimeWithSpeedVariable}.Time
  ) {desc=15 fewer instructions on low end, materialFunction=/Clyde/Environments/KelpLinen/Fortnite_SkyTwistFunction}
  n44 = Multiply(A=@n43.Rotated UVs, B=@n43.Rotated UVs)
  n45 = Multiply(A=TextureCoordinate(), B=ScalarParameter() {defaultValue=10, group=Spirals, parameterName=Cloud-TileMaster2}) {constB=0.5}
  n46 = Constant() {r=1}
  n47 = ScalarParameter() {defaultValue=-0.005, group=Spirals, parameterName=SwirlPanSpeed2}
  n48 = ScalarParameter() {defaultValue=0.2, group=Spirals, parameterName=Cloud-SpeedMaster2}
  n49 = MaterialFunctionCall(
   Flow Direction=Constant2Vector() {g=0.2, r=-1}
   Time offset=ScalarParameter() {defaultValue=2.01, group=Spirals, parameterName=Swirl2Offset}
   UVS * size=Multiply(A=Subtract(A=@n45, B=@n46), B=ScalarParameter() {defaultValue=2, group=Spirals, parameterName=SwirlTile2}) {constB=0.5}
   SineAnimSpeed=ScalarParameter() {defaultValue=0.5, group=Spirals, parameterName=SwirlCurlSpeed2}
   Rotation amount modulator=Add(A=Multiply() {constB=-0.15}, B=ScalarParameter() {defaultValue=0.02, group=Spirals, parameterName=SwirlAmount2})
   un frac Time=MaterialFunctionCall(Speed=Multiply(A=@n47, B=@n48)) {materialFunction=&TimeWithSpeedVariable}.Time
  ) {desc=15 fewer instructions on low end, materialFunction=&Fortnite_SkyTwistFunction}
  n50 = Multiply(A=@n49.Rotated UVs, B=@n49.Rotated UVs)
  NamedRerouteDeclaration(None=Add(A=Divide(A=@n38) {constB=4}, B=Add(A=@n44, B=@n50))) {name=Skybox_Warping, nodeColor=<rgba 0 1 0.5078 1>}
  n51 = Constant()
  n52 = NamedRerouteDeclaration(None=ScalarParameter() {defaultValue=0.016, group=Night Sky, parameterName=NightSkyRotation}) {name=NightSkyRotation, nodeColor=<rgba 0 1 0.4375 1>}
  NamedRerouteDeclaration(None=Add(A=TextureCoordinate(), B=AppendVector(A=@n51, B=@n52))) {name=star pan, nodeColor=<rgba 0.3594 0 1 1>}
  n53 = TextureCoordinate()
  n54 = AppendVector(
   A=ScalarParameter() {defaultValue=0.88, group=Kelp_Glow, parameterName=Backdrop_GlowX_Position}
   B=ScalarParameter() {defaultValue=0.872, group=Kelp_Glow, parameterName=Backdrop_GlowY_Position}
  )
  n55 = ScalarParameter() {defaultValue=0.3637, group=Kelp_Glow, parameterName=Backdrop_Size}
  n56 = ScalarParameter() {defaultValue=0.7575, group=Kelp_Glow, parameterName=Backdrop_Density}
  n57 = Subtract(A=@n56, B=ScalarParameter() {group=Beat, parameterName=Density_BeatOffset})
  n58 = ScalarParameter() {defaultValue=2.5429, group=Kelp_Glow, parameterName=Backdrop_Multi}
  n59 = Multiply(
   A=MaterialFunctionCall(UVs=@n53, CenterPosition=@n54, Radius=@n55, Density=LinearInterpolate(A=@n57, B=@n56, Alpha=@n10) {constA=2}) {materialFunction=&RadialGradientExponential}.RadialGradientExponential
   B=@n58
  )
  n60 = Multiply(
   A=MaterialFunctionCall(
    UVs=@n53
    CenterPosition=@n54
    Radius=ScalarParameter() {defaultValue=0.217, group=Kelp_Glow, parameterName=Backdrop_Size_Inner, sortPriority=33}
    Density=ScalarParameter() {defaultValue=3.3648, group=Kelp_Glow, parameterName=Backdrop_Density_Inner, sortPriority=33}
   ) {materialFunction=&RadialGradientExponential}.RadialGradientExponential
   B=ScalarParameter() {defaultValue=1.0391, group=Kelp_Glow, parameterName=Backdrop_Multi_inner, sortPriority=33}
  )
  NamedRerouteDeclaration(None=LinearInterpolate(B=Add(A=@n59, B=@n60), Alpha=ScalarParameter() {group=Kelp_Glow, parameterName=Kelp_Backdrop_enabled}) {constB=0}) {name=Kelp_Backdrop_Glow, nodeColor=<rgba 0.1484 0 1 1>}
  n61 = TextureCoordinate()
  n62 = NamedRerouteDeclaration(None=TextureObject() {samplerType=SAMPLERTYPE_LinearColor, texture=/Clyde/Environments/KelpLinen/T_Clyde_SkyDome_Stars_Sharp}) {name=StarTile, nodeColor=<rgba 0 0.1562 1 1>}
  n63 = TextureSample(UVs=Multiply(A=@n61, B=ScalarParameter() {defaultValue=16, group=Kelp_Stars, parameterName=Kelp star_size_small}), Tex=@n62) {samplerSource=SSM_Wrap_WorldGroupSettings, samplerType=SAMPLERTYPE_LinearColor}
  n64 = Multiply(A=@n61, B=ScalarParameter() {defaultValue=12, group=Kelp_Stars, parameterName=Kelp star_size_middle})
  n65 = Multiply(A=@n61, B=ScalarParameter() {defaultValue=8, group=Kelp_Stars, parameterName=Kelp star_size_big})
  n66 = Multiply(
   A=Max(
    A=TextureSample(UVs=@n64, Tex=@n62) {samplerSource=SSM_Wrap_WorldGroupSettings, samplerType=SAMPLERTYPE_LinearColor}.R
    B=TextureSample(UVs=@n65, Tex=@n62) {samplerSource=SSM_Wrap_WorldGroupSettings, samplerType=SAMPLERTYPE_LinearColor}.R
   )
   B=ScalarParameter() {group=Kelp_Stars, parameterName=use FG Big Star}
  )
  n67 = Multiply(
   A=Power(Base=NamedRerouteUsage(), Exp=ScalarParameter() {defaultValue=2, group=Kelp_Stars, parameterName=Kelp_Star_Pow})
   B=ScalarParameter() {group=Kelp_Stars, parameterName=Kelp_GalaxyMask_Str}
  )
  NamedRerouteDeclaration(None=Multiply(A=Max(A=@n63.R, B=@n66), B=OneMinus(None=@n67))) {name=kelp_Star, nodeColor=<rgba 1 0 0.2812 1>}
  n68 = TextureCoordinate()
  n69 = MaterialFunctionCall(UVs=@n68, "Rotation Angle (0-1)"=ScalarParameter() {defaultValue=1.1259, group=Kelp, parameterName=Galaxy_rotation}) {materialFunction=&CustomRotator}
  n70 = VectorParameter() {
   group=Kelp
   parameterName=Galaxy_Controls
   channelNames {a=offset y, b=offset x, g=scale y, r=scale x}
   defaultValue=<rgba 0.5 0.5 0.104 0.096>
  }
  n71 = AppendVector(A=@n70.scale x, B=@n70.scale x)
  n72 = Add(
   A=Add(
    A=MaterialFunctionCall(UVs=@n69.Rotated Values, Texture Scale=@n71) {materialFunction=&ScaleUVsByCenter}.UVs
    B=AppendVector(A=@n70.offset x, B=@n70.offset x)
   )
   B=Multiply(A=NamedRerouteUsage(), B=ScalarParameter() {group=Spirals, parameterName=warping str})
  )
  n73 = TextureSample(UVs=@n72) {samplerType=SAMPLERTYPE_LinearColor, texture=/Clyde/Environments/KelpLinen/T_Clyde_Skydome_Update}
  NamedRerouteDeclaration(None=@n73.RGB) {name=Galaxy_Mask, nodeColor=<rgba 0 0.6016 1 1>}
  n74 = TextureCoordinate()
  n75 = AppendVector(
   A=ScalarParameter() {defaultValue=0.88, group=Kelp_Clouds, parameterName=Cloud_Radius_X_Position}
   B=ScalarParameter() {defaultValue=0.9758, group=Kelp_Clouds, parameterName=CloudRadius_Y_Position}
  )
  n76 = ScalarParameter() {defaultValue=0.6757, group=Kelp_Clouds, parameterName=CloudRadius_Size}
  n77 = TextureCoordinate()
  n78 = ScalarParameter() {defaultValue=5.8278, group=Kelp_Clouds, parameterName=Cloud_Radius_Noise}
  n79 = NamedRerouteUsage()
  n80 = Rotator(Coordinate=Add(A=Multiply(A=@n77, B=@n78) {constB=0.5}, B=Multiply(A=@n79) {constB=5})) {speed=0.002}
  n81 = Subtract(
   A=Multiply(
    A=MaterialFunctionCall(UVs=@n74, CenterPosition=@n75, Radius=@n76) {materialFunction=&RadialGradientExponential}.RadialGradientExponential
    B=ScalarParameter() {defaultValue=2.729, group=Kelp_Clouds, parameterName=CloudSize_Multi}
   )
   B=Multiply(
    A=TextureSample(UVs=@n80) {samplerSource=SSM_Wrap_WorldGroupSettings, texture=/Game/TimeOfDay/TODM_Sequence/Materials/Sky/Clouds/T_Clouds_02_M}.RGB
    B=ScalarParameter() {defaultValue=0.9805, group=Kelp_Clouds, parameterName=Cloud_Radius_Noise_Multi}
   )
  )
  n82 = NamedRerouteUsage()
  NamedRerouteDeclaration(None=Saturate(None=Multiply(A=@n81, B=@n82))) {name=Kelp_Cloud_Mask, nodeColor=<rgba 1 0 0.3984 1>}
  NamedRerouteDeclaration(None=MaterialFunctionCall() {materialFunction=/ClydeCosmetics/Environments/KelpLinen/MF_KelpLinen_LocationFade}.Output) {name=Kelp_Play_Area, nodeColor=<rgba 0 0.625 1 1>}
  NamedRerouteDeclaration(None=@n13) {name=Backdrop_Cloud_Light, nodeColor=<rgba 1 0 0.0703 1>}
  NamedRerouteDeclaration(None=@n60) {name=Backdrop_InnerMAsk, nodeColor=<rgba 0.7344 1 0 1>}
  NamedRerouteDeclaration(None=MaterialFunctionCall!(In) {materialFunction=/ClydeCosmetics/Environments/KelpLinen/MF_Kelp_AudioReactive}.Beat_Tracker) {name=Music Beat, nodeColor=<rgba 0.2891 1 0 1>}
  n83 = MaterialFunctionCall(
   UVs=ComponentMask(None=CameraPositionWS()) {g=true, r=true}
   CenterPosition=ComponentMask(None=Constant3Vector() {constant=<rgba -31136 -105728 -1920 1>}) {g=true, r=true}
   Radius=Add(A=Constant() {r=23054.5273}, B=ScalarParameter() {defaultValue=6237.2627, group=Kelp_Location, parameterName=Cloud_Fade_offset})
   Density=Constant() {r=10}
  ) {materialFunction=&RadialGradientExponential}
  NamedRerouteDeclaration(None=Multiply(
    A=Saturate(None=@n83.RadialGradientExponential)
    B=CollectionParameter() {collection=&FortniteMaterialParameters, parameterName=SlurpActive}
   )) {name=Kelp_Cloud_Fade, nodeColor=<rgba 1 0 0.2812 1>}
  n84 = Multiply(
   A=MaterialFunctionCall(UVs=@n53, CenterPosition=@n54, Radius=@n55, Density=@n56) {materialFunction=&RadialGradientExponential}.RadialGradientExponential
   B=@n58
  )
  NamedRerouteDeclaration(None=LinearInterpolate(B=Add(A=@n84, B=@n60), Alpha=ScalarParameter() {group=Kelp_Glow, parameterName=Kelp_Backdrop_enabled}) {constB=0}) {name=Kelp_Backdrop_Glow_Clouds, nodeColor=<rgba 0 1 0.0156 1>}
  NamedRerouteDeclaration(None=@n5) {name=Kelp_Backdrop_Glow_Final, nodeColor=<rgba 0.8281 1 0 1>}
  n85 = CollectionParameter() {collection=&FortniteMaterialParameters, parameterName=PlayerPosition}
  n86 = CollectionParameter() {collection=&FortniteMaterialParameters, parameterName=PlayerPosition}
  n87 = ComponentMask(None=@n86) {b=true, g=true, r=true}
  n88 = Constant3Vector() {constant=<rgba -27852 -50384 -63335.4922 1>}
  n89 = ScalarParameter() {
   defaultValue=500
   desc="Use this to make a element fade faster or slower than other objects when the camera moves thru the zone.  Tweak carefully.  "
   group=Kelp_Location
   parameterName=Reflection_Object_Offset
  }
  n90 = Add(A=Constant() {desc=Default Radius - Don't Touch unless fade distance for entire zone is changing, r=131922.5}, B=@n89)
  n91 = ScalarParameter() {defaultValue=0.85, desc=Default for old values0.5, group=Kelp_Location, parameterName=Density}
  n92 = Saturate(None=Multiply(A=SphereMask(A=@n87, B=@n88, Radius=@n90, Hardness=@n91), B=Constant() {r=2}))
  n93 = If(A=ComponentMask(None=@n85) {b=true, g=true, r=true}, B=Constant3Vector(), "A > B"=@n92, "A == B"=Constant(), "A < B"=@n92)
  NamedRerouteDeclaration(None=@n93) {name=lumen mask, nodeColor=<rgba 0 0.5312 1 1>}
  n94 = MaterialFunctionCall() {materialFunction=/Game/TimeOfDay/TODM_Sequence/Materials/Sky/MaterialFunctions/MF_Helios_SunMoon}
  NamedRerouteDeclaration(None=@n94.MoonMask) {name=MoonMask, nodeColor=<rgba 1 0.8672 0 1>}
  n95 = VectorParameter() {group=Kelp, parameterName=Galaxy_Color_Highlight, defaultValue=<rgba 0.8632 0.4852 0.7305 1>}
  n96 = Multiply(A=@n95.RGB, B=Power(Base=@n73.RGB, Exp=ScalarParameter() {defaultValue=2.0792, group=Kelp, parameterName=Galaxy_Pow}))
  n97 = Multiply(A=Time(), B=ScalarParameter() {defaultValue=0.05, group=Kelp, parameterName=FoamWaveSpeed})
  n98 = MaterialFunctionCall() {materialFunction=&Pi}
  n99 = AppendVector(A=Sine(None=@n97), B=Sine(None=Add(A=@n97, B=@n98.Result) {constB=0.5}))
  n100 = ScalarParameter() {defaultValue=0.2, group=Kelp, parameterName=FoamWaveSineMotion}
  n101 = TextureSample(UVs=Panner(Coordinate=TextureCoordinate()) {speedX=0.1, speedY=-0.02}) {
   samplerSource=SSM_Wrap_WorldGroupSettings
   samplerType=SAMPLERTYPE_Grayscale
   texture=/Game/TimeOfDay/TODM_Sequence/Materials/Sky/Textures/LowResBlurredNoise_Radish_LF_Root
  }
  n102 = Multiply(A=AppendVector(A=Multiply(A=@n99, B=@n100), B=Constant()), B=@n101.RGB)
  n103 = Multiply(
   A=Power(Base=@n73.RGB, Exp=ScalarParameter() {defaultValue=2, group=Kelp, parameterName=Galaxy_Twinkles_Pow})
   B=MaterialFunctionCall(Default=Saturate(None=@n102), Low/Mobile=Constant()) {materialFunction=&MF_LowMobile_PathingSwitch}.Result
  )
  n104 = Multiply(A=@n103, B=Multiply(A=@n95.RGB, B=ScalarParameter() {defaultValue=1, group=Kelp, parameterName=Light_Panner_Multi}) {constB=0.4})
  n105 = Saturate(None=@n103)
  n106 = VectorParameter() {group=Kelp, parameterName=Galaxy_Color_Base, defaultValue=<rgba 0.8632 0.4852 0.7305 1>}
  n107 = Power(Base=@n73.RGB, Exp=ScalarParameter() {defaultValue=0.8573, group=Kelp, parameterName=Galaxy_Pow_background})
  n108 = ScalarParameter() {defaultValue=0.88, group=Kelp, parameterName=Galaxy_FadeX}
  n109 = ScalarParameter() {defaultValue=0.872, group=Kelp, parameterName=Galaxy_FadeY}
  n110 = LinearInterpolate(
   B=Add(A=LinearInterpolate(A=@n96, B=@n104, Alpha=@n105), B=Multiply(A=@n106.RGB, B=@n107))
   Alpha=MaterialFunctionCall(
    UVs=@n68
    CenterPosition=AppendVector(A=@n108, B=@n109)
    Radius=ScalarParameter() {defaultValue=0.297, group=Kelp, parameterName=Galaxy_FadeSize}
    Density=ScalarParameter() {defaultValue=6.6873, group=Kelp, parameterName=Galaxy_FadeDensity}
   ) {materialFunction=&RadialGradientExponential}.RadialGradientExponential
  ) {constB=0}
  n111 = ScalarParameter() {defaultValue=0.5, desc=used to make the galaxy not bloom out in inner glow area, group=Kelp, parameterName=Galaxy_Backdrop_Multi}
  n112 = LinearInterpolate(A=@n110, B=Multiply(A=@n111, B=@n110), Alpha=Saturate(None=NamedRerouteUsage()))
  n113 = Constant() {r=0.75}
  n114 = Multiply(
   A=NamedRerouteUsage()
   B=MaterialFunctionCall(Value_Morning=@n113, Value_Day=Constant(), Value_Evening=@n113, Value_Night=Constant() {r=1}) {materialFunction=&Scalar_Time_of_Day}.Result
  )
  n115 = Multiply(A=ScalarParameter() {group=Kelp, parameterName=Kelp_galaxies_enabled}, B=NamedRerouteUsage())
  n116 = Multiply(A=@n115, B=NamedRerouteUsage())
  NamedRerouteDeclaration(None=LinearInterpolate(B=Add(A=@n112, B=@n114), Alpha=ReflectionCapturePassSwitch!(Default=@n115, Reflection=@n116)) {constB=0}) {name=Kelp_Additions, nodeColor=<rgba 0.7812 1 0 1>}
  n117 = Constant3Vector() {constant=<rgba 43.7499 39.9786 39.1023 1>}
  n118 = NamedRerouteUsage()
  n119 = Multiply(A=@n118, B=ScalarParameter() {defaultValue=20, parameterName=Start1}) {constB=5}
  n120 = TextureObject() {samplerType=SAMPLERTYPE_LinearColor, texture=/Game/TimeOfDay/TODM_Sequence/Materials/Sky/Textures/SkyNightMap_output}
  n121 = Multiply(A=@n118, B=ScalarParameter() {defaultValue=8, parameterName=StarMask})
  n122 = NamedRerouteUsage()
  n123 = Multiply(A=@n122, B=ScalarParameter() {defaultValue=12, parameterName=Start2}) {constB=5}
  n124 = Multiply(A=@n122, B=ScalarParameter() {defaultValue=15, parameterName=StarMask2})
  n125 = Add(
   A=Multiply(
    A=TextureSample(UVs=@n119, Tex=@n120) {samplerType=SAMPLERTYPE_LinearColor}.G
    B=TextureSample(UVs=@n121, Tex=@n120) {samplerSource=SSM_Wrap_WorldGroupSettings, samplerType=SAMPLERTYPE_LinearColor}.R
   )
   B=Multiply(
    A=TextureSample(UVs=@n123, Tex=@n120) {samplerSource=SSM_Wrap_WorldGroupSettings, samplerType=SAMPLERTYPE_LinearColor}.G
    B=TextureSample(UVs=@n124, Tex=@n120) {samplerSource=SSM_Wrap_WorldGroupSettings, samplerType=SAMPLERTYPE_LinearColor}.B
   )
  )
  n126 = Multiply(A=NamedRerouteUsage(), B=Constant() {r=5})
  n127 = Multiply(
   A=LinearInterpolate(B=Multiply(A=@n117, B=@n125), Alpha=NamedRerouteUsage())
   B=LinearInterpolate(Alpha=Saturate(None=@n126)) {constA=1, constB=0.6}
  )
  n128 = TextureSample(UVs=Panner(Coordinate=TextureCoordinate() {uTiling=18, vTiling=18}) {speedX=0.36}) {
   samplerSource=SSM_Wrap_WorldGroupSettings
   samplerType=SAMPLERTYPE_Grayscale
   texture=&LowResBlurredNoise_Radish_LF_Root
  }
  n129 = MaterialFunctionCall(Default=LinearInterpolate(Alpha=@n128.R) {constA=0.05}, Low/Mobile=Constant() {r=0.8}) {materialFunction=&MF_LowMobile_PathingSwitch}
  n130 = NamedRerouteUsage()
  n131 = NamedRerouteUsage()
  NamedRerouteDeclaration(None=LinearInterpolate(A=Multiply(A=@n127, B=@n129.Result), Alpha=Multiply(A=@n130, B=@n131)) {constB=0}) {name=MilkywayStars, nodeColor=<rgba 0.7812 0 1 1>}
  NamedRerouteDeclaration(None=@n94.MoonHaloMask) {name=MoonHaloMask, nodeColor=<rgba 0.4297 1 0 1>}
  NamedRerouteDeclaration(None=@n94.MoonActive) {name=MoonActive, nodeColor=<rgba 0 0.2734 1 1>}
  n132 = MaterialFunctionCall(UVs=NamedRerouteUsage(), Texture Scale=ScalarParameter() {defaultValue=1, group=Night Sky, parameterName=MIkyWaySize}) {materialFunction=&ScaleUVsByCenter}
  n133 = TextureSample(
   UVs=Multiply(
    A=NamedRerouteDeclaration(None=@n132.UVs) {name=milkyway_uv, nodeColor=<rgba 0 1 0.8359 1>}
    B=ScalarParameter() {defaultValue=5500, group=Night Sky, parameterName=StarmaskFrequency}
   )
   Tex=@n120
  ) {
   samplerSource=SSM_Wrap_WorldGroupSettings
   samplerType=SAMPLERTYPE_LinearColor
   texture=/Game/TimeOfDay/TODM_Sequence/Materials/Sky/Textures/T_milkyway_StarMask
  }
  NamedRerouteDeclaration(None=@n133.B) {name=StarMask, nodeColor=<rgba 1 0.8438 0 1>}
  n134 = Multiply(A=Time(), B=ScalarParameter() {defaultValue=0.1, desc=Sudo how long the pulse takes, group=Beat, parameterName=Time_Multi}) {constB=0.2}
  Power(Base=ConstantBiasScale(None=Sine(None=@n134)), Exp=ScalarParameter() {defaultValue=10, parameterName=Exp}) {constExponent=4}
  n135 = ScalarParameter() {defaultValue=1, group=Kelp_Clouds, parameterName=Kelp_Clouds_Opacity}
  n136 = NamedRerouteDeclaration(None=LinearInterpolate(A=@n7.CloudAlpha, B=Multiply(A=@n7.CloudAlpha, B=@n135), Alpha=@n9)) {name=cloudAlpha, nodeColor=<rgba 1 0 0.4687 1>}
  n137 = NamedRerouteUsage()
  n138 = Multiply(A=NamedRerouteUsage()) {constB=1.5}
  n139 = Multiply(A=@n94.Result, B=OneMinus(None=Saturate(None=@n138)))
  n140 = Saturate(None=CollectionParameter() {collection=&FortniteMaterialParameters, parameterName=SkyBoxClouds})
  n141 = Multiply(A=CollectionParameter() {collection=/Game/Athena/SafeZone/MPC_SafeZoneClouds, parameterName=CloudStartingBias}) {constB=2}
  n142 = TextureCoordinate()
  n143 = Constant() {r=0.5}
  n144 = Constant()
  n145 = Multiply(
   A=OneMinus(None=Saturate(None=@n141))
   B=SphereMask(
    A=SphereMask(A=@n142, B=@n143, Radius=@n143, Hardness=@n144)
    B=Constant() {r=0.085}
    Radius=Constant() {r=0.07}
    Hardness=Constant() {r=0.1128}
   )
  )
  n146 = NamedRerouteUsage()
  n147 = LinearInterpolate(
   A=NamedRerouteUsage()
   B=VectorParameter() {parameterName=StormHorizonGradientColor, defaultValue=<rgba 0.0907 0.0645 0.125 1>}.RGB
   Alpha=Multiply(A=NamedRerouteDeclaration(None=@n145) {name=StormGradientMask, nodeColor=<rgba 0.4297 1 0 1>}, B=OneMinus(None=@n146))
  ) {constAlpha=1}
  n148 = Add(
   A=QualitySwitch(Default=@n139, Epic=LinearInterpolate(A=@n94.Result, B=@n139, Alpha=@n140))
   B=QualitySwitch(Default=NamedRerouteUsage(), High=@n147, Epic=@n147)
  )
  n149 = Add(A=NamedRerouteUsage(), B=@n148)
  n150 = MaterialFunctionCall!(
   Low _ PC Low and Med=@n148
   "High _ XB1, PS4"=@n149
   "Epic _ PC High, XSX, PS5"=@n149
   --------
   "Mobile Low _ PC Perf, Android"=@n148
   Mobile Med _ Switch=@n148
  ) {materialFunction=/Game/Global/MaterialFunctions/TechArt/MF_Fortnite_QualitySwitch_Vector3}
  n151 = LinearInterpolate(A=@n150.Result, B=@n7.CloudColor, Alpha=Saturate(None=Subtract(A=@n136, B=@n137))) {constAlpha=0, constB=0}
  n152 = CollectionParameter() {collection=&FortniteMaterialParameters, parameterName=SkyBoxClouds}
  n153 = QualitySwitch(Default=@n151, Epic=LinearInterpolate(A=@n150.Result, B=@n151, Alpha=Saturate(None=@n152)) {constAlpha=1})
  n154 = Multiply(A=@n153, B=NamedRerouteUsage())
  n155 = ScalarParameter() {defaultValue=1, group=Kelp_Glow, parameterName=Backdrop_Reflection_Multi}
  n156 = ReflectionCapturePassSwitch(Default=@n154, Reflection=Multiply(A=@n154, B=Multiply(A=@n155, B=@n93)))
  n157 = NamedRerouteUsage()
  n158 = NamedRerouteUsage()
  n159 = NamedRerouteUsage()
  n160 = NamedRerouteUsage()
  n161 = Add(A=LinearInterpolate(A=@n153, B=@n156, Alpha=Multiply(A=@n157, B=@n158)), B=Multiply(A=@n156, B=Multiply(A=@n159, B=@n160)))
  n162 = VectorParameter() {parameterName=Reflection_Color, defaultValue=<rgba 1 1 1 1>}
  n163 = Multiply(A=ScalarParameter() {defaultValue=1, parameterName=Reflect_Color_Boost}, B=@n161)
  n164 = ScalarParameter() {defaultValue=0.2, parameterName=ReflectionDesaturation_Morning}
  n165 = ScalarParameter() {defaultValue=0.2, parameterName=ReflectionDesaturation_Day}
  n166 = ScalarParameter() {defaultValue=0.2, parameterName=ReflectionDesaturation_Evening}
  n167 = ScalarParameter() {defaultValue=0.45, parameterName=ReflectionDesaturation_Night}
  ReflectionCapturePassSwitch(
   Default=@n161
   Reflection=Desaturation(
    None=Multiply(A=@n162.RGB, B=@n163)
    Fraction=MaterialFunctionCall(Value_Morning=@n164, Value_Day=@n165, Value_Evening=@n166, Value_Night=@n167) {materialFunction=&Scalar_Time_of_Day}.Result
   )
  )
  Rotator(Coordinate=@n72) {speed=0.02}
  Saturate(None=@n34)
  n168 = ComponentMask(None=CameraPositionWS()) {b=true}
  ShadingPathSwitch(Default=Constant(), Mobile=AppendVector(A=Constant2Vector(), B=Multiply(A=@n168) {constB=0.9}))
  n169 = ScalarParameter() {defaultValue=0.1, desc=Sudo how long the pulse takes, group=Beat, parameterName=Time_Multi}
  Sine(None=Multiply(A=Time(), B=Multiply(A=@n169) {constB=2, desc=Number of waves in water}) {constB=0.2})
  TextureSample() {texture=/Clyde/Environments/KelpLinen/T_Clyde_Nebula_Noise}
 ]
}
